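# Translucent unlit material for Niagara mesh particles that draws an animated shockwave sphere, filed as a tentacle directional attack FX reference.
# Panning chromatic and magnetic-wave textures, warped by cell noise and view-space normals, combine with a perturbed Fresnel that blends inner and outer colors and shapes opacity.
# Vertex offset adds normal inflation and a pinch relative to particle center, while dynamic parameters feed pinch, panning and inner opacity.
# Default-quality shading adds HDR remapping and depth fade; low, forward and mobile skip both.
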Material /PistolCanary/Assets/FX/TentacleDirectionalAttack/References/M_Fortanium_Shockwave_Sphere
graph {
 properties {
  bUsedWithNiagaraMeshParticles=true
  blendMode=BLEND_Translucent
  pixelDepthOffsetMode=PDOM_Legacy
  shadingModel=MSM_Unlit
  usedShadingModels=Unlit
 }
 nodes [
  n1 = Multiply(A=VectorParameter() {parameterName=WPOPinch, defaultValue=<rgba 1 1 0.5 1>}.RGB, B=Subtract(A=ParticlePositionWS(), B=WorldPosition().XYZ))
  n2 = Power(Base=OneMinus(None=NamedRerouteUsage())) {constExponent=3}
  Add(
   A=Multiply(A=VertexNormalWS(), B=ScalarParameter() {parameterName=VertexOffset})
   B=Multiply(A=Multiply(A=@n1, B=@n2), B=NamedRerouteUsage())
  )
  n3 = ParticleColor()
  n4 = Multiply(A=@n3.RGB, B=NamedRerouteUsage())
  n5 = ScalarParameter() {defaultValue=5, parameterName=HDRInensity}
  n6 = MaterialFunctionCall(LDR Input=Clamp(None=@n4), HDR Tint=@n4, HDR Intensity Multiplier=@n5, Desaturate Luminance=Constant() {r=0.1}) {materialFunction=&DeriveHDRfromLDR}
  n7 = Multiply(A=@n4, B=@n5)
  MaterialFunctionCall(Emissive=ShadingPathSwitch(Default=QualitySwitch(Default=@n6.RemappedRange, Low=@n7), Forward=@n7, Mobile=@n7)) {materialFunction=&MF_FX_EmissiveAdjustments}
  n8 = DynamicParameter() {defaultValue=<rgba 0 1 1 0.5>, paramNames [WPOPinch,Tile,PanAdd,InnerOpacity]}
  NamedRerouteDeclaration(None=@n8.WPOPinch) {name=WPOPinchParam, nodeColor=<rgba 1 0.0703 0 1>}
  n9 = TextureCoordinate()
  n10 = Constant() {r=0.5}
  n11 = NamedRerouteDeclaration(None=SphereMask(A=ComponentMask(None=@n9) {g=true}, B=@n10, Radius=@n10, Hardness=ScalarParameter() {parameterName=EndFadeHardness}) {hardnessPercent=0}) {name=Gradient, nodeColor=<rgba 0 1 0.5781 1>}
  n12 = MaterialFunctionCall(Float2=TextureCoordinate() {uTiling=2}) {materialFunction=&BreakOutFloat2Components}
  n13 = Abs(None=Subtract(A=@n12.G) {constB=0.5})
  n14 = Saturate(None=Multiply(A=@n13) {constB=5})
  n15 = Multiply(A=Power(Base=@n13) {constExponent=3}) {constB=3}
  n16 = Add(A=Multiply(A=Time(), B=ScalarParameter() {defaultValue=1, parameterName=PanSpeed}), B=@n8.PanAdd)
  n17 = TextureSample(UVs=Panner(Coordinate=AppendVector(A=@n12.R, B=@n15), Time=@n16) {speedX=0.05, speedY=0.3}) {texture=/Game/Effects/Fort_Effects/Textures/Tile/T_Smooth_Tile_Chromatic}
  n18 = ScalarParameter() {defaultValue=0.6, parameterName=NoiseDistortion}
  n19 = Panner(Coordinate=Multiply(A=TextureCoordinate(), B=ScalarParameter() {defaultValue=1, parameterName=LumpsTile}), Time=@n16) {speedX=-0.05, speedY=0.1}
  n20 = ComponentMask(None=Transform(None=VertexNormalWS()) {transformSourceType=TRANSFORMSOURCE_World, transformType=TRANSFORM_View}) {g=true, r=true}
  n21 = Power(Base=LinearInterpolate(
    A=TextureSample(UVs=@n19) {samplerType=SAMPLERTYPE_LinearGrayscale, texture=/Game/Effects/Fort_Effects/Textures/Tile/T_CellNoise_Tile_M}.G
    B=Abs(None=@n20)
   ) {constAlpha=0.7}) {constExponent=3}
  n22 = VectorParameter() {parameterName=DistortionTile, defaultValue=<rgba 1 4 0 0>}
  n23 = Multiply(A=Add(A=Multiply(A=@n18, B=@n21), B=TextureCoordinate()), B=AppendVector(A=@n22.R, B=@n22.R))
  n24 = TextureSample(UVs=MaterialFunctionCall(XY=Panner(Coordinate=@n23, Time=@n16) {speedX=0.05, speedY=0.2}) {materialFunction=&Swizzle}.YX) {texture=/Game/Athena/Environments/Nexus/Textures/T_ZeroPoint_MagneticWaves_Blurred}
  n25 = VertexNormalWS()
  n26 = Multiply(A=@n24.B, B=@n25)
  n27 = Fresnel(ExponentIn=ScalarParameter() {defaultValue=1, parameterName=EdgeFadeExp}, Normal=Subtract(A=@n25, B=Multiply(A=@n26) {constB=0.5})) {baseReflectFraction=0, exponent=1}
  n28 = Saturate(None=LinearInterpolate(Alpha=@n27) {constA=1.7, constB=-0.3})
  n29 = Multiply(A=Power(Base=OneMinus(None=@n28)) {constExponent=1.7}, B=NamedRerouteUsage())
  n30 = Add(A=Max(A=Multiply(A=@n14, B=@n17.RGB), B=@n24.RGB), B=Multiply(A=@n29) {constB=3})
  n31 = Multiply(
   A=Max(A=Saturate(None=@n30), B=@n29)
   B=LinearInterpolate(
    A=VectorParameter() {parameterName=OuterColor, defaultValue=<rgba 0.121 0.142 1 0>}.RGB
    B=VectorParameter() {parameterName=InnerColor, defaultValue=<rgba 0.4495 0.1031 0.6 1>}.RGB
    Alpha=@n27
   )
  )
  n32 = Multiply(A=@n28, B=Saturate(None=LinearInterpolate(A=@n8.InnerOpacity, Alpha=@n27) {constA=0.2, constB=1.4}))
  n33 = Multiply(
   A=Multiply(A=Multiply(A=@n3.A, B=@n11)) {constB=2}
   B=Multiply(A=NamedRerouteDeclaration(None=@n31) {name=EmissiveTexture, nodeColor=<rgba 0.4531 1 0 1>}, B=Power(Base=@n32))
  )
  n34 = ScalarParameter() {defaultValue=100, parameterName=DephFade}
  ShadingPathSwitch(Default=QualitySwitch(Default=DepthFade(Opacity=@n33, FadeDistance=@n34), Low=@n33), Forward=@n33, Mobile=@n33)
 ]
}
